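; Atlantis roof building actor Blueprint that carries a decorative flora mesh component and hides it on lower-end platforms.
; On BeginPlay it calls the parent, then, only where cosmetic events can execute, sets that flora mesh hidden in game on Nintendo Switch or mobile, without propagating to children.
; The component is named for the A_02 flora variant although the asset is A_01, and the construction script is empty.
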(event EventBeginPlay
  (|CallParentFunction)
  (Networking|CanExecuteCosmeticEvents
    (:True
      (Development|SetHiddeninGame (Variables|Default|GetSM_Atlantis_Roof_A_02_Flora_01) (or (TouchPlatform|IsNintendoSwitch) (TouchPlatform|IsMobileGame)) false))))

(fn ConstructionScript)
